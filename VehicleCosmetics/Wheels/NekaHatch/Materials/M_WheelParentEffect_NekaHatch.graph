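# Unlit parent material for NekaHatch wheels, drawing a stylized animated hatched look into emissive.
# Fresnel, normal maps and a view-based term threshold panning screen-space hatch textures, with vertex red blending rim and tire maps.
# RGB masks add dark detail, a scrolling gradient and accent color, while a Painted switch derives tones from Paint Color.
# The result passes through vehicle damage and de-rez transition functions.
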Material /VehicleCosmetics/Wheels/NekaHatch/Materials/M_WheelParentEffect_NekaHatch
graph {
 properties {
  bAutomaticallySetUsageInEditor=false
  bUseMaterialAttributes=true
  bUsedWithMeshDeformer=true
  bUsedWithSkeletalMesh=true
  shadingModel=MSM_Unlit
  usedShadingModels=Unlit
 }
 nodes [
  n1 = Fresnel() {baseReflectFraction=0.3, exponent=3}
  n2 = Power(Base=@n1, Exp=ScalarParameter() {defaultValue=32, parameterName=Frsnel power})
  n3 = Clamp(None=ConstantBiasScale(None=@n1))
  n4 = WorldPosition()
  n5 = DotProduct(A=Normalize(VectorInput=CameraVectorWS()), B=Normalize(VectorInput=Abs(None=@n4.XYZ)))
  n6 = TextureSampleParameter2D() {parameterName=N_Rim, samplerType=SAMPLERTYPE_Normal, texture=/VehicleCosmetics/Wheels/NekaHatch/Textures/T_NekaHatch_Rim_N}
  n7 = TextureSampleParameter2D() {parameterName=N_Tier, samplerType=SAMPLERTYPE_Normal, texture=/VehicleCosmetics/Wheels/NekaHatch/Textures/T_NekaHatch_Tire_N}
  n8 = VertexColor()
  n9 = ConstantBiasScale(None=Fresnel(Normal=LinearInterpolate(A=@n6.RGB, B=@n7.RGB, Alpha=@n8.R)) {exponent=1}) {bias=0.2, scale=1.2}
  n10 = ScalarParameter() {defaultValue=32, parameterName=Frsnel power}
  n11 = Subtract(A=Clamp(None=OneMinus(None=@n5)), B=Clamp(None=Power(Base=@n9, Exp=@n10)))
  n12 = ScalarParameter() {defaultValue=0.3, parameterName=Hatch Distance}
  n13 = ComponentMask(None=ScreenPosition().ViewportUV) {g=true, r=true}
  n14 = Panner(Coordinate=Multiply(A=@n13, B=ScalarParameter() {defaultValue=1, parameterName=UV Hatch Tiling})) {speedX=0.2}
  n15 = TextureSample(UVs=@n14) {texture=/VehicleCosmetics/Wheels/_SharedTextures/FX/skin_Jer_RGB}
  n16 = If(A=@n11, B=Constant() {r=0.75}, "A > B"=Constant() {r=1}, "A == B"=@n15.G, "A < B"=@n15.G)
  n17 = Min(A=@n16, B=TextureSample(UVs=@n14) {texture=&skin_Jer_RGB}.R)
  n18 = TextureSampleParameter2D() {parameterName=Neka Hatch Hub, samplerType=SAMPLERTYPE_LinearColor, texture=/VehicleCosmetics/Wheels/NekaHatch/Textures/T_NekaHatch_Rim_RGB}
  n19 = TextureSampleParameter2D() {
   parameterName=Neka Hatch Thread
   samplerType=SAMPLERTYPE_LinearColor
   texture=/VehicleCosmetics/Wheels/NekaHatch/Textures/T_NekaHatch_Tire_RGB
  }
  n20 = Subtract(
   A=Subtract(A=Max(A=@n2, B=@n3), B=If(A=@n11, B=@n12, "A > B"=@n16, "A == B"=@n17, "A < B"=@n17))
   B=LinearInterpolate(A=@n18.R, B=@n19.R, Alpha=@n8.R)
  )
  n21 = Clamp(None=ConstantBiasScale(None=@n20) {bias=0.1, scale=1})
  n22 = VectorParameter() {parameterName=Paint Color, defaultValue=<rgba 0 0.34 0.12 1>}
  n23 = ConstantBiasScale(None=@n22.RGB) {bias=0, scale=0.075}
  n24 = Constant() {r=-2}
  n25 = ConstantBiasScale(None=@n22.RGB)
  n26 = Constant() {r=-2}
  n27 = StaticSwitchParameter(True=Clamp(None=Desaturation(None=@n25, Fraction=@n26)), False=ScalarParameter() {defaultValue=0.05, parameterName=Flat Color}) {parameterName=Painted}
  n28 = Clamp(None=Desaturation(None=ConstantBiasScale(None=@n22.RGB) {bias=0.3, scale=1}, Fraction=Constant() {r=-2}))
  n29 = Multiply(A=@n13, B=Constant() {r=2})
  n30 = LinearInterpolate(
   A=@n22.RGB
   B=@n28
   Alpha=TextureSampleParameter2D(UVs=Panner(Coordinate=@n29) {speedY=0.5}) {parameterName=Airmage Gradient V, texture=/VehicleCosmetics/Wheels/_SharedTextures/FX/airmage_Gradiant_V}.R
  )
  n31 = LinearInterpolate(A=@n18.G, B=@n19.G, Alpha=@n8.R)
  n32 = VectorParameter() {parameterName=Color 2, defaultValue=<rgba 0.8 0.71 0.06 1>}
  n33 = LinearInterpolate(
   A=Clamp(None=Desaturation(None=@n23, Fraction=@n24))
   B=LinearInterpolate(
    A=LinearInterpolate(A=@n27, B=@n30, Alpha=@n31)
    B=StaticSwitchParameter(True=@n28, False=@n32.RGB) {parameterName=Painted}
    Alpha=LinearInterpolate(A=@n18.B, B=@n19.B, Alpha=@n8.R)
   )
   Alpha=Clamp(None=@n20)
  )
  n34 = Multiply(
   A=ScalarParameter() {defaultValue=1, parameterName=Brightness}
   B=StaticSwitchParameter(True=Multiply(A=@n21, B=@n33), False=@n33) {parameterName=Painted}
  )
  MaterialFunctionCall(In=MaterialFunctionCall(Material=SetMaterialAttributes(Emissive Color=@n34) {attributeSetTypes [#1]}) {materialFunction=&MF_BaseVehicle_Damage}.Result) {materialFunction=&MF_VehicleDeRez_Transition}
 ]
}
